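# Translucent unlit mesh-particle material for large ground fog, built from panning storm cloud textures with bump-offset depth.
# Sun/moon direction dotted with the normal map adds highlights, tinted with shadow color by time of day.
# Opacity combines cloud mask, vertex red, particle alpha and mask-varied depth fade, scaled by a time-of-day cosine.
# Low quality and mobile paths use a simple shadow-to-highlight color blend.
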Material /Figment_S07/Effects/Fort_Effects/Materials/Smoke/M_Ground_Fog_Large
graph {
 properties {
  bTangentSpaceNormal=false
  bUsedWithMeshParticles=true
  blendMode=BLEND_Translucent
  pixelDepthOffsetMode=PDOM_Legacy
  shadingModel=MSM_Unlit
  usedShadingModels=Unlit
 }
 nodes [
  n1 = VectorParameter() {parameterName=HighlightColor, defaultValue=<rgba 1.1375 1.3521 1.4 0>}
  n2 = CollectionParameter() {
   collection=&FortniteMaterialParameters
   parameterName=SunAndMoonModelDirectionalVector
  }
  n3 = TextureCoordinate()
  n4 = VectorParameter() {parameterName=Tiling, defaultValue=<rgba 5 1.5 0 0>}
  n5 = AppendVector(A=@n4.R, B=@n4.R)
  n6 = ScalarParameter() {defaultValue=1, parameterName=Tiling}
  n7 = ScalarParameter() {defaultValue=1, parameterName=Distance}
  n8 = LinearInterpolate(Alpha=Clamp(None=Multiply(A=@n7) {constB=4})) {constA=1, constB=0.35}
  n9 = Panner(
   Coordinate=Multiply(A=Multiply(A=@n3, B=@n5), B=Multiply(A=@n6, B=@n8))
   Time=Multiply(A=Time(), B=ScalarParameter() {defaultValue=0.7, parameterName=PanSpeed})
  ) {speedX=0.001, speedY=-0.02}
  n10 = TextureSample(UVs=@n9) {texture=/Game/VisualThreatManager/StormVisuals/Textures/2ndInfectedMapLinear}
  n11 = BumpOffset(
   Coordinate=@n9
   Height=Multiply(A=LinearInterpolate(Alpha=@n10.G) {constA=-1}, B=ScalarParameter() {defaultValue=0.5, parameterName=BumpOffset})
  )
  n12 = Clamp(None=DotProduct(
    A=ComponentMask(None=@n2) {b=true, g=true, r=true}
    B=TextureSample(UVs=@n11) {samplerType=SAMPLERTYPE_Normal, texture=/Game/VisualThreatManager/StormVisuals/Textures/2ndInfectedMapNormal}.RGB
   ))
  n13 = ScalarParameter() {defaultValue=2, parameterName=Spec Highlight Power}
  n14 = Multiply(A=@n1.RGB, B=Multiply(A=Power(Base=@n12, Exp=@n13) {constExponent=20})) {constB=0}
  n15 = VectorParameter() {parameterName=EveningColor, defaultValue=<rgba 2 0.9946 0.3395 0>}
  n16 = MaterialFunctionCall(
   Color_Morning=@n15.RGB
   Color_Evening=@n15.RGB
   Color_Night=VectorParameter() {parameterName=NightColor, defaultValue=<rgba 0.5576 0.4597 1 0>}.RGB
  ) {materialFunction=&Color_Time_of_Day}
  n17 = LinearInterpolate(A=@n16.Result) {constAlpha=0.4}
  n18 = VectorParameter() {parameterName=ShadowColor, defaultValue=<rgba 0.4344 0.4457 0.6 0>}
  n19 = TextureSample(UVs=@n11) {texture=&2ndInfectedMapLinear}
  n20 = Multiply(A=LinearInterpolate(A=@n18.RGB, B=@n1.RGB, Alpha=@n19.G)) {constB=0.7}
  n21 = QualitySwitch(Default=Add(A=Multiply(A=@n14, B=@n16.Result), B=Multiply(A=@n17, B=@n18.RGB)), Low=@n20)
  n22 = ParticleColor()
  Multiply(A=FeatureLevelSwitch(Default=@n21, ES2=@n20, ES3_1=@n20, SM4_REMOVED=@n21), B=@n22.RGB)
  n23 = LinearInterpolate(Alpha=@n19.G) {constA=1, constB=0.1}
  n24 = ScalarParameter() {defaultValue=15000, parameterName=DepthFadeDist}
  n25 = Multiply(A=Multiply(A=VertexColor().R, B=@n19.G), B=DepthFade(FadeDistance=Multiply(A=@n23, B=@n24)))
  n26 = ScalarParameter() {defaultValue=0.4, parameterName=Opacity}
  n27 = CollectionParameter() {collection=&FortniteMaterialParameters, parameterName=TimeOfDay_0-1}
  n28 = Power(Base=ConstantBiasScale(None=Cosine(None=@n27))) {constExponent=0.6}
  Multiply(A=@n22.A, B=Multiply(A=Multiply(A=@n25, B=@n26), B=LinearInterpolate(Alpha=@n28) {constA=0.5}))
 ]
}
